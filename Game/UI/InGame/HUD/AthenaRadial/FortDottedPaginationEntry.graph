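# Dotted pagination entry button for a HUD radial list, where the first page can appear as a labeled "LOCKER" marker instead of a circle dot.
# ListViewVisuals swaps marker material, border brush, padding and label visibility by first-item and selected state, with optional drop-shadow variants.
# IsFirstListItem checks the first displayed list entry, gated by Fort.Rollback.GroupEmotePickerCategories equaling 1.
# Selection events play Select or reversed Deselect animations, with timers and sequencer events to refresh visuals.
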WidgetBlueprint /Game/UI/InGame/HUD/AthenaRadial/FortDottedPaginationEntry
animations [
 Select {
  binding_count=4
  length_seconds=0.3
  object=&self:Select
  sequence {
   bindings [
    ScaleBox_Text {
     id=#1
     possessed_class=ScaleBox
     tracks [{
       type=MovieSceneFloatTrack
       display_name=User Specified Scale
       sections [{type=MovieSceneFloatSection, end_frame=4, start_frame=0, channels [None {type=MovieSceneScriptingFloatChannel, keys [[0,0],[4,0.85]]}]}]
      }]
    }
    Text_Locker {
     id=#2
     possessed_class=UEFN_TextBlock_C
     tracks [
      {
       type=MovieSceneColorTrack
       display_name=Color and Opacity
       sections [{
         type=MovieSceneColorSection
         end_frame=6
         start_frame=0
         channels [
          Color.R {type=MovieSceneScriptingFloatChannel, const [1,1]}
          Color.G {type=MovieSceneScriptingFloatChannel, const [1,1]}
          Color.B {type=MovieSceneScriptingFloatChannel, const [1,1]}
          Color.A {type=MovieSceneScriptingFloatChannel, keys [[0,0],[6,1]]}
         ]
        }]
      }
      {
       type=MovieScene2DTransformTrack
       sections [{
         type=MovieScene2DTransformSection
         end_frame=6
         start_frame=0
         channels [
          Translation.X {type=MovieSceneScriptingFloatChannel, const [0,1]}
          Translation.Y {type=MovieSceneScriptingFloatChannel, const [0,1]}
          Angle {type=MovieSceneScriptingFloatChannel, const [0,1]}
          Scale.X {type=MovieSceneScriptingFloatChannel, keys [[0,0.75],[6,1]]}
          Scale.Y {type=MovieSceneScriptingFloatChannel, keys [[0,0.75],[6,1]]}
          Shear.X {type=MovieSceneScriptingFloatChannel, const [0,1]}
          Shear.Y {type=MovieSceneScriptingFloatChannel, const [0,1]}
         ]
        }]
      }
     ]
    }
    Image_CircleEntry {
     id=#3
     possessed_class=Image
     tracks [{
       type=MovieSceneWidgetMaterialTrack
       sections [{
         type=MovieSceneParameterSection
         end_frame=6
         start_frame=0
         channels [
          OuterCircle {type=MovieSceneScriptingFloatChannel, keys [[0,0.0339],[4.1443,0.7297],[5.9397,0.8939]]}
          OuterBorder {type=MovieSceneScriptingFloatChannel, keys [[0,0.1281],[4.549,0.0332],[5.9397,0.0381]]}
         ]
        }]
      }]
    }
    RoundedBorder {
     id=#4
     possessed_class=CommonBorder
     tracks [{
       type=MovieSceneWidgetMaterialTrack
       sections [{
         type=MovieSceneParameterSection
         end_frame=6
         start_frame=0
         channels [
          Opacity {type=MovieSceneScriptingFloatChannel, keys [[0,0],[6,1]]}
          Size {type=MovieSceneScriptingFloatChannel, keys [[0,0.91],[6,0.8]]}
         ]
        }]
      }]
    }
   ]
   tracks [{
     type=MovieSceneEventTrack
     sections [{
       type=MovieSceneEventTriggerSection
       channels [None {type=MovieSceneScriptingEventChannel, const ["<Struct 'MovieSceneEvent' (0x4679464a280) {}>", 1]}]
      }]
    }]
  }
 }
 Deselect {
  binding_count=4
  length_seconds=0.201
  object=&self:Deselect
  sequence {
   bindings [
    Text_Locker {
     id=#2
     possessed_class=UEFN_TextBlock_C
     tracks [
      {
       type=MovieSceneColorTrack
       display_name=Color and Opacity
       sections [{
         type=MovieSceneColorSection
         end_frame=4
         start_frame=0
         channels [
          Color.R {type=MovieSceneScriptingFloatChannel, const [1,1]}
          Color.G {type=MovieSceneScriptingFloatChannel, const [1,1]}
          Color.B {type=MovieSceneScriptingFloatChannel, const [1,1]}
          Color.A {type=MovieSceneScriptingFloatChannel, keys [[0,0],[4.019,1]]}
         ]
        }]
      }
      {
       type=MovieScene2DTransformTrack
       sections [{
         type=MovieScene2DTransformSection
         end_frame=4
         start_frame=0
         channels [
          Translation.X {type=MovieSceneScriptingFloatChannel, const [0,1]}
          Translation.Y {type=MovieSceneScriptingFloatChannel, const [0,1]}
          Angle {type=MovieSceneScriptingFloatChannel, const [0,1]}
          Scale.X {type=MovieSceneScriptingFloatChannel, keys [[0,0.75],[4.019,1]]}
          Scale.Y {type=MovieSceneScriptingFloatChannel, keys [[0,0.75],[4.019,1]]}
          Shear.X {type=MovieSceneScriptingFloatChannel, const [0,1]}
          Shear.Y {type=MovieSceneScriptingFloatChannel, const [0,1]}
         ]
        }]
      }
     ]
    }
    Image_CircleEntry {
     id=#3
     possessed_class=Image
     tracks [{
       type=MovieSceneWidgetMaterialTrack
       sections [{
         type=MovieSceneParameterSection
         end_frame=4
         start_frame=0
         channels [
          OuterCircle {type=MovieSceneScriptingFloatChannel, keys [[0,0.0339],[2.7757,0.7297],[3.9783,0.8939]]}
          OuterBorder {type=MovieSceneScriptingFloatChannel, keys [[0,0],[3.047,0.0332],[3.9783,0.0381]]}
         ]
        }]
      }]
    }
    RoundedBorder {
     id=#4
     possessed_class=CommonBorder
     tracks [{
       type=MovieSceneWidgetMaterialTrack
       sections [{
         type=MovieSceneParameterSection
         end_frame=4
         start_frame=0
         channels [
          Opacity {type=MovieSceneScriptingFloatChannel, keys [[0,0],[4.019,1]]}
          Size {type=MovieSceneScriptingFloatChannel, keys [[0,0.91],[4.019,0.8]]}
         ]
        }]
      }]
    }
    ScaleBox_Text {
     id=#5
     possessed_class=ScaleBox
     tracks [{
       type=MovieSceneFloatTrack
       display_name=User Specified Scale
       sections [{type=MovieSceneFloatSection, end_frame=3, start_frame=0, channels [None {type=MovieSceneScriptingFloatChannel, keys [[0,0],[3.047,1]]}]}]
      }]
    }
   ]
   tracks [{
     type=MovieSceneEventTrack
     sections [{
       type=MovieSceneEventTriggerSection
       channels [None {type=MovieSceneScriptingEventChannel, const ["<Struct 'MovieSceneEvent' (0x46794649380) {}>", 1]}]
      }]
    }]
  }
 }
]
graphs [
 EventGraph {nodes [
   OnEntryGenerated() {is=title K2Node_CustomEvent}
   > n1 = IsFirstListItem!(execute, self) -> (then, IsFirstListItem:bool)
   > ListViewVisuals!(execute, self, IsFirstListItem:bool=@n1, IsSelected:bool=GetSelected!(self:Common Button Base Object Reference) -> (ReturnValue:bool)) -> (then)
   OnAnimationComplete() {is=title K2Node_CustomEvent}
   > n2 = IsFirstListItem()
   > ListViewVisuals(IsFirstListItem=@n2, IsSelected=GetSelected())
   ResetSelectedState() {is=title K2Node_CustomEvent}
   > SetSelectedInternal!(execute, self:Common Button Base Object Reference, bInSelected:bool=false, bAllowSound:bool=false, bBroadcast:bool=false) -> (then)
   > ListViewVisuals(IsFirstListItem=true, IsSelected=false)
   Event On Deselected()
   > StopAllAnimations(self:User Widget Object Reference)
   > n3 = IsFirstListItem()
   n4 = GetNumericCVar!(CVar:String=Fort.Rollback.GroupEmotePickerCategories) -> (OutValue:float, ReturnValue:bool)
   > n5 = Branch(Condition=AND Boolean(A=@n3, B="Equal (Float)"(A=@n4, B=1)))
   > PlayAnimation(InAnimation=Get Deselect() -> (Deselect:Widget Animation Object Reference), PlayMode=Reverse)
   > Set Timer by Event(Delegate=Create Event(), Time=0.2, bLooping=false, bMaxOncePerFrame=true)
   ListViewVisuals(execute=@n5.else, IsFirstListItem=@n3, IsSelected=GetSelected())
   Event Construct()
   > n6 = IsFirstListItem()
   > ListViewVisuals(IsFirstListItem=@n6, IsSelected=false)
   Event On Selected()
   > StopAllAnimations()
   > PlayAnimation(InAnimation=Get Select() -> (Select:Widget Animation Object Reference))
   > Set Timer by Event(Delegate=Create Event(), Time=0.3, bLooping=false, bMaxOncePerFrame=true)
   Event PreConstruct()
   n7 = Get PreviewFirstPage() -> (PreviewFirstPage:bool)
   > ListViewVisuals(IsFirstListItem=@n7, IsSelected=@n7)
   n8 = Event On Item Selection Changed!() -> (OutputDelegate, then, bIsSelected:bool) {is=title K2Node_Event}
   > n9 = Branch(Condition="Equal (Float)"(A=GetNumericCVar(CVar=Fort.Rollback.GroupEmotePickerCategories), B=1))
   n10 = IsFirstListItem(execute=@n9.else)
   > ListViewVisuals(IsFirstListItem=@n10, IsSelected=@n8.bIsSelected)
   > PlayAnimation(InAnimation=Get Select())
   Event On Entry Released() {is=title K2Node_Event}
   > ListViewVisuals(IsFirstListItem=false, IsSelected=false)
  ]}
 IsFirstListItem {nodes [
   IsFirstListItem() {is=title K2Node_FunctionEntry}
   > n1 = Is Valid(InputObject=GetOwningListView!(UserListEntry:User List Entry Interface) -> (ReturnValue:List View Base Object Reference))
   Return Node!(execute=@n1.Is Not Valid, IsFirstListItem:bool="Equal (Float)"(A=GetNumericCVar(CVar=Fort.Rollback.GroupEmotePickerCategories), B=1))
   n2 = "Get (a copy)"(Array:Array of User Widget Object References=GetDisplayedEntryWidgets!(self:List View Base Object Reference=GetOwningListView()) -> (ReturnValue:Array of User Widget Object References)) -> (Output:User Widget Object Reference)
   n3 = Self-Reference()
   n4 = GetNumericCVar(CVar=Fort.Rollback.GroupEmotePickerCategories)
   > Return Node(IsFirstListItem=AND Boolean(A="Equal (Object)"(A=@n2, B=@n3), B="Equal (Float)"(A=@n4, B=1)))
  ]}
 ListViewVisuals {nodes [
   n1 = ListViewVisuals!() -> (then, IsFirstListItem:bool=~, IsSelected:bool=~) {is=title K2Node_FunctionEntry}
   > n2 = Set IsSelectedLocal(IsSelectedLocal:bool=@n1.IsSelected) -> (Output_Get:bool)
   > Set PreviewFirstPage(PreviewFirstPage:bool=@n2) -> (Output_Get:bool)
   > n3 = Branch(Condition=@n1)
   n4 = Get Text_Locker() -> (Text_Locker:Text Block Object Reference)
   > SetShadowOffset!(
    execute
    self:Text Object Reference=@n4
    InShadowOffset:Vector 2D Structure=Select(
     Option 0:Vector 2D Structure
     Option 1:Vector 2D Structure=Get TextShadowOffset() -> (TextShadowOffset:Vector 2D Structure)
     Index=Get DropShadow() -> (DropShadow:bool)
    ) -> (ReturnValue:Vector 2D Structure)
   ) -> (then)
   > Set ShadowColorAndOpacity!(
    execute
    ShadowColorAndOpacity:Linear Color Structure=Get Text_ShadowColor() -> (Text_ShadowColor:Linear Color Structure)
    self:Text Object Reference=@n4
   ) -> (then, Output_Get:Linear Color Structure)
   n5 = Get Image_CircleEntry() -> (Image_CircleEntry:Image Object Reference)
   > SetBrushFromMaterial(
    self=@n5
    Material=Select(
     Option 0:Material Instance Object Reference=Get FirstPage_MatInst() -> (FirstPage_MatInst:Material Instance Object Reference)
     Option 1:Material Instance Object Reference=Get FirstPage_MatInst_DropShadow() -> (FirstPage_MatInst_DropShadow:Material Instance Object Reference)
     Index=Get DropShadow()
    ) -> (ReturnValue:Material Instance Object Reference)
   )
   > n6 = Branch(Condition=Get IsSelectedLocal() -> (IsSelectedLocal:bool))
   n7 = Get RoundedBorder() -> (RoundedBorder:Common Border Object Reference)
   > SetBrush(
    self:Border Object Reference=@n7
    InBrush=Select(
     Option 0:Slate Brush Structure=Get ActiveBorderBrush() -> (ActiveBorderBrush:Slate Brush Structure)
     Option 1:Slate Brush Structure=Get ActiveBorderBrush_DropShadow() -> (ActiveBorderBrush_DropShadow:Slate Brush Structure)
     Index=Get DropShadow()
    ) -> (ReturnValue:Slate Brush Structure)
   )
   > SetPadding!(
    execute
    self:Border Object Reference=@n7
    InPadding_Left:float=10
    InPadding_Top:float=3
    InPadding_Right:float=10
    InPadding_Bottom:float=3
   ) -> (then)
   > n8 = SetVisibility(self=Get Text_Locker(), InVisibility=Visible)
   Return Node!(execute=[@n8.then, @n10.then])
   n9 = SetBrush(execute=@n6.else, self=Get RoundedBorder(), InBrush=Get InactiveBorderBrush() -> (InactiveBorderBrush:Slate Brush Structure))
   n10 = SetVisibility(execute=[@n9.then, @n12.then], self=Get Text_Locker(), InVisibility=Collapsed)
   SetBrushFromMaterial(
    execute=@n3.else
    self=@n5
    Material=Select(
     Option 0:Material Interface Object Reference=/Game/Athena/HUD/LTM/Mash/Materials/M_UI_CirclePage_Marker_Selected_Unselected
     Option 1:Material Interface Object Reference=/Game/Athena/HUD/LTM/Mash/Materials/M_UI_CirclePage_Marker_Selected
     Index=Get IsSelectedLocal()
    ) -> (ReturnValue:Material Interface Object Reference)
   )
   n11 = Get RoundedBorder()
   > SetBrush(self=@n11, InBrush=Get InactiveBorderBrush())
   > n12 = SetPadding(self=@n11, InPadding_Left=0, InPadding_Top=3, InPadding_Right=0, InPadding_Bottom=3)
  ]}
 Sequencer Events {nodes [
   SequenceEvent() {is=title K2Node_CustomEvent}
   > n1 = IsFirstListItem() {is=title K2Node_CallFunction}
   > ListViewVisuals(IsFirstListItem=@n1, IsSelected=GetSelected()) {is=title K2Node_CallFunction}
  ]}
]
variables [
 ActiveBorderBrush {
  type=SlateBrush
  default {
   bIsDynamicallyLoaded=false
   drawAs=Image
   imageType=FullColor
   mirroring=NoMirror
   resourceName=None
   resourceObject=/Game/Athena/HUD/LTM/Mash/Materials/MI_UI_Border_PageMarker
   tiling=NoTile
   imageSize=<32 32>
   margin=<ltrb 0 0 0 0>
   outlineSettings {
    bUseBrushTransparency=false
    roundingType=HalfHeightRadius
    width=0
    color {colorUseRule=UseColor_Specified, specifiedColor=<rgba 0 0 0 0>}
    cornerRadii=<0 0 0 0>
   }
   tintColor {colorUseRule=UseColor_Specified, specifiedColor=<rgba 1 1 1 1>}
   uVRegion {bIsValid=false, max=<0 0>, min=<0 0>}
  }
 }
 ActiveBorderBrush_DropShadow {
  type=SlateBrush
  default {
   bIsDynamicallyLoaded=false
   drawAs=Image
   imageType=FullColor
   mirroring=NoMirror
   resourceName=None
   resourceObject=/Game/Athena/HUD/LTM/Mash/Materials/MI_UI_Border_PageMarker_DropShadow
   tiling=NoTile
   imageSize=<32 32>
   margin=<ltrb 0 0 0 0>
   outlineSettings {
    bUseBrushTransparency=false
    roundingType=HalfHeightRadius
    width=0
    color {colorUseRule=UseColor_Specified, specifiedColor=<rgba 0 0 0 0>}
    cornerRadii=<0 0 0 0>
   }
   tintColor {colorUseRule=UseColor_Specified, specifiedColor=<rgba 1 1 1 1>}
   uVRegion {bIsValid=false, max=<0 0>, min=<0 0>}
  }
 }
 CirclePage_MatInst {type=MaterialInstance, default=&M_UI_CirclePage_Marker_Selected}
 DropShadow {type=boolean, default=true}
 FirstPage_MatInst {type=MaterialInstance, default=/Game/Athena/HUD/LTM/Mash/Materials/MI_UI_LockerPage_Marker}
 FirstPage_MatInst_DropShadow {type=MaterialInstance, default=/Game/Athena/HUD/LTM/Mash/Materials/MI_UI_LockerPage_Marker_DropShadow}
 InactiveBorderBrush {
  type=SlateBrush
  default {
   bIsDynamicallyLoaded=false
   drawAs=NoDrawType
   imageType=NoImage
   mirroring=NoMirror
   resourceName=None
   resourceObject=null
   tiling=NoTile
   imageSize=<32 32>
   margin=<ltrb 0 0 0 0>
   outlineSettings {
    bUseBrushTransparency=false
    roundingType=HalfHeightRadius
    width=0
    color {colorUseRule=UseColor_Specified, specifiedColor=<rgba 0 0 0 0>}
    cornerRadii=<0 0 0 0>
   }
   tintColor {colorUseRule=UseColor_Specified, specifiedColor=<rgba 1 1 1 1>}
   uVRegion {bIsValid=false, max=<0 0>, min=<0 0>}
  }
 }
 PreviewFirstPage {type=boolean}
 Text_ShadowColor {type=LinearColor, default=<rgba 0 0 0 0.502>}
 TextShadowOffset {type=Vector2D, default=<1.15 1.5>}
]
widgets {
 info {inheritedWidgetCount=1, parentClass=&CommonButtonLegacy, rootWidgetClass=&Overlay, widgetCount=7}
 widgets [
  Overlay_0 {widgetClassPath=&Overlay, widget=Overlay_0, properties {visibility=HitTestInvisible}}
  RoundedBorder {
   widgetClassPath=&CommonBorder
   bIsVariable=true
   parent=Overlay_0
   widget=RoundedBorder
   properties {
    background {
     bIsDynamicallyLoaded=false
     drawAs=NoDrawType
     imageType=FullColor
     mirroring=NoMirror
     resourceName=None
     resourceObject=null
     tiling=NoTile
     imageSize=<64 64>
     margin=<ltrb 0 0 0 0>
     outlineSettings {
      bUseBrushTransparency=false
      roundingType=HalfHeightRadius
      width=0
      color {colorUseRule=UseColor_Specified, specifiedColor=<rgba 0 0 0 0>}
      cornerRadii=<0 0 0 0>
     }
     tintColor {colorUseRule=UseColor_Specified, specifiedColor=<rgba 1 1 1 1>}
     uVRegion {bIsValid=false, max=<0 0>, min=<0 0>}
    }
    padding=<ltrb 10 3 10 3>
   }
   slot {class=OverlaySlot, properties {horizontalAlignment=HAlign_Fill, verticalAlignment=VAlign_Fill}}
  }
  HorizontalBox {
   widgetClassPath=&HorizontalBox
   bIsVariable=true
   parent=RoundedBorder
   widget=HorizontalBox
   slot {class=BorderSlot, properties {padding=<ltrb 10 3 10 3>}}
  }
  Image_CircleEntry {
   widgetClassPath=&Image
   bIsVariable=true
   parent=HorizontalBox
   widget=Image_CircleEntry
   properties {brush {
     bIsDynamicallyLoaded=false
     drawAs=Image
     imageType=NoImage
     mirroring=NoMirror
     resourceName=None
     resourceObject=&M_UI_CirclePage_Marker_Selected_Unselected
     tiling=NoTile
     imageSize=<64 64>
     margin=<ltrb 0 0 0 0>
     outlineSettings {
      bUseBrushTransparency=false
      roundingType=HalfHeightRadius
      width=0
      color {colorUseRule=UseColor_Specified, specifiedColor=<rgba 0 0 0 0>}
      cornerRadii=<0 0 0 0>
     }
     tintColor {colorUseRule=UseColor_Specified, specifiedColor=<rgba 1 1 1 1>}
     uVRegion {bIsValid=false, max=<0 0>, min=<0 0>}
    }}
   slot {class=HorizontalBoxSlot, properties {horizontalAlignment=HAlign_Center, verticalAlignment=VAlign_Center, padding=<ltrb -14 -14 -14 -14>}}
  }
  ScaleBox_Text {
   widgetClassPath=&ScaleBox
   parent=HorizontalBox
   widget=ScaleBox_Text
   properties {stretch=UserSpecified}
   slot {class=HorizontalBoxSlot, properties {horizontalAlignment=HAlign_Center, verticalAlignment=VAlign_Center, size {sizeRule=Fill, value=1}}}
  }
  Text_Locker {
   widgetClassPath=&UEFN_TextBlock
   bIsVariable=true
   parent=ScaleBox_Text
   widget=Text_Locker
   properties {
    justification=Center
    text=LOCKER
    font {
     bForceMonospaced=false
     bMaterialIsStencil=false
     fontMaterial=null
     fontObject=&BurbankSmall
     letterSpacing=0
     monospacedWidth=1
     size=18
     skewAmount=0
     typefaceFontName=Black
     outlineSettings {
      bApplyOutlineToDropShadows=false
      bMiteredCorners=false
      bSeparateFillAlpha=false
      outlineMaterial=null
      outlineSize=0
      outlineColor=<rgba 0 0 0 1>
     }
    }
    shadowColorAndOpacity=<rgba 0 0 0 1>
    shadowOffset=<1.5 2>
   }
   slot {class=ScaleBoxSlot}
  }
  InputActionWidget {widgetClassPath=&CommonActionWidget, bInherited=true, widget=null}
 ]
}
