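# Widget Blueprint of FortDottedPagination that shows a row of pagination dots, located with the in-game HUD's Athena radial UI.
# The inherited ListView_Dots is a horizontal, non-focusable list that ignores the mouse wheel and uses FortDottedPaginationEntry widgets as dots.
# On initialization it binds to the list's entry-generated event, casts each new widget to FortDottedPaginationEntry and calls that entry's OnEntryGenerated.
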WidgetBlueprint /Game/UI/InGame/HUD/AthenaRadial/FortDottedPagination
graphs [EventGraph {nodes [
   n1 = OnEntryGenerated!() -> (OutputDelegate, then, Widget:User Widget Object Reference) {is=title K2Node_CustomEvent}
   > n2 = Cast To FortDottedPaginationEntry!(execute, Object=@n1.Widget) -> (then, CastFailed, AsFort Dotted Pagination Entry:Fort Dotted Pagination Entry Object Reference) {is=title K2Node_DynamicCast}
   > OnEntryGenerated(self:Fort Dotted Pagination Entry Object Reference=@n2) {is=title K2Node_CallFunction}
   Event OnInitialized()
   > Bind Event to On Entry Generated(self:List View Base Object Reference=Get ListView_Dots() -> (ListView_Dots:List View Object Reference), Delegate=@n1) {is=title K2Node_AddDelegate}
  ]}]
variables []
widgets {
 info {inheritedWidgetCount=1, parentClass=/Script/FortniteUI.FortDottedPagination, rootWidgetClass=&CommonListView, widgetCount=1}
 widgets [ListView_Dots {
   widgetClassPath=&CommonListView
   bInherited=true
   bIsVariable=true
   widget=ListView_Dots
   properties {
    bIsFocusable=false
    clipping=Inherit
    consumeMouseWheel=Never
    entryWidgetClass=&/FortDottedPaginationEntry.FortDottedPaginationEntry_C
    numDesignerPreviewEntries=8
    orientation=Orient_Horizontal
   }
  }]
}
