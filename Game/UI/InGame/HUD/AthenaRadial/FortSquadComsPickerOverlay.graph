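# Widget Blueprint for the in-game HUD squad comms radial picker, built on FortPickerOverlay.
# A centered fixed-size stack layers a shadow, pie background and selection highlight, all material-driven, under a FortSlottedRadialMenu with custom compass ordering and FortSelectableRadialEntry items.
# The PlaySelectionAnimation event plays SelectionSet, which animates the highlight material's inner and outer radius and the highlight's opacity.
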WidgetBlueprint /Game/UI/InGame/HUD/AthenaRadial/FortSquadComsPickerOverlay
animations [SelectionSet {
  binding_count=1
  length_seconds=0.5
  object=&self:SelectionSet
  sequence {bindings [Image_RadialHighlight {
     id=#1
     possessed_class=Image
     tracks [
      {
       type=MovieSceneWidgetMaterialTrack
       sections [{
         type=MovieSceneParameterSection
         channels [
          OuterRadius_0-1 {type=MovieSceneScriptingFloatChannel, keys [[0,0.97],[10,1]]}
          InnerRadius_0-1 {type=MovieSceneScriptingFloatChannel, keys [[0,0.5603],[10,0.6071]]}
         ]
        }]
      }
      {
       type=MovieSceneFloatTrack
       display_name=Render Opacity
       sections [{type=MovieSceneFloatSection, channels [None {type=MovieSceneScriptingFloatChannel, keys [[0,0],[1.557,1]]}]}]
      }
     ]
    }]}
 }]
graphs [
 Cosmetic {nodes [
   Event PlaySelectionAnimation() {is=title K2Node_Event}
   > PlayAnimation(InAnimation=Get SelectionSet() -> (SelectionSet:Widget Animation Object Reference))
  ]}
 EventGraph {nodes [Reroute Node(InputPin:Wildcard) -> (OutputPin:Wildcard)]}
]
variables []
widgets {
 info {inheritedWidgetCount=3, parentClass=/Script/FortniteUI.FortPickerOverlay, rootWidgetClass=&Overlay, widgetCount=8}
 widgets [
  Overlay_1 {widgetClassPath=&Overlay, widget=Overlay_1}
  SizeBox_7 {
   widgetClassPath=&SizeBox
   parent=Overlay_1
   widget=SizeBox_7
   properties {bOverride_HeightOverride=true, bOverride_WidthOverride=true, heightOverride=700, widthOverride=700}
   slot {class=OverlaySlot, properties {horizontalAlignment=HAlign_Center, verticalAlignment=VAlign_Center}}
  }
  Overlay_0 {widgetClassPath=&Overlay, parent=SizeBox_7, widget=Overlay_0, slot {class=SizeBoxSlot}}
  Image_Radial_BG {
   widgetClassPath=&Image
   parent=Overlay_0
   widget=Image_Radial_BG
   properties {
    visibility=HitTestInvisible
    brush {
     bIsDynamicallyLoaded=false
     drawAs=Image
     imageType=NoImage
     mirroring=NoMirror
     resourceName=None
     resourceObject=/Game/UI/Items/Materials/MI-AthenaPieShadow
     tiling=NoTile
     imageSize=<128 128>
     margin=<ltrb 0 0 0 0>
     outlineSettings {
      bUseBrushTransparency=false
      roundingType=HalfHeightRadius
      width=0
      color {colorUseRule=UseColor_Specified, specifiedColor=<rgba 0 0 0 0>}
      cornerRadii=<0 0 0 0>
     }
     tintColor {colorUseRule=UseColor_Specified, specifiedColor=<rgba 1 1 1 1>}
     uVRegion {bIsValid=false, max=<0 0>, min=<0 0>}
    }
    colorAndOpacity=<rgba 0.0052 0.0052 0.0056 0.4>
    renderTransform {angle=0, scale=<1 1>, shear=<0 0>, translation=<20 20>}
   }
   slot {class=OverlaySlot, properties {horizontalAlignment=HAlign_Fill, verticalAlignment=VAlign_Fill}}
  }
  Image_Radial {
   widgetClassPath=&Image
   bInherited=true
   parent=Overlay_0
   widget=Image_Radial
   properties {
    visibility=HitTestInvisible
    brush {
     bIsDynamicallyLoaded=false
     drawAs=Image
     imageType=NoImage
     mirroring=NoMirror
     resourceName=None
     resourceObject=/Game/UI/Items/Materials/MI-AthenaPieMenu
     tiling=NoTile
     imageSize=<128 128>
     margin=<ltrb 0 0 0 0>
     outlineSettings {
      bUseBrushTransparency=false
      roundingType=HalfHeightRadius
      width=0
      color {colorUseRule=UseColor_Specified, specifiedColor=<rgba 0 0 0 0>}
      cornerRadii=<0 0 0 0>
     }
     tintColor {colorUseRule=UseColor_Specified, specifiedColor=<rgba 1 1 1 1>}
     uVRegion {bIsValid=false, max=<0 0>, min=<0 0>}
    }
    colorAndOpacity=<rgba 0.0273 0.0273 0.0273 1>
   }
   slot {class=OverlaySlot, properties {horizontalAlignment=HAlign_Fill, verticalAlignment=VAlign_Fill}}
  }
  CommonBorder_0 {
   widgetClassPath=&CommonBorder
   parent=Overlay_0
   widget=CommonBorder_0
   properties {padding=<ltrb 0 0 0 0>, renderTransform {angle=0, scale=<1 1>, shear=<0 0>, translation=<0 -9>}}
   slot {class=OverlaySlot, properties {horizontalAlignment=HAlign_Fill, verticalAlignment=VAlign_Fill}}
  }
  Image_RadialHighlight {
   widgetClassPath=&Image
   bInherited=true
   bIsVariable=true
   parent=CommonBorder_0
   widget=Image_RadialHighlight
   properties {
    visibility=HitTestInvisible
    brush {
     bIsDynamicallyLoaded=false
     drawAs=Image
     imageType=NoImage
     mirroring=NoMirror
     resourceName=None
     resourceObject=/Game/UI/Items/Materials/MI-AthenaPieMenu_Selection
     tiling=NoTile
     imageSize=<32 32>
     margin=<ltrb 0 0 0 0>
     outlineSettings {
      bUseBrushTransparency=false
      roundingType=HalfHeightRadius
      width=0
      color {colorUseRule=UseColor_Specified, specifiedColor=<rgba 0 0 0 0>}
      cornerRadii=<0 0 0 0>
     }
     tintColor {colorUseRule=UseColor_Specified, specifiedColor=<rgba 1 1 1 1>}
     uVRegion {bIsValid=false, max=<0 0>, min=<0 0>}
    }
   }
   slot {class=BorderSlot, properties {padding=<ltrb 0 0 0 0>}}
  }
  RadialMenu_Picker {
   widgetClassPath=/Game/Athena/HUD/ContextMenus/FortSlottedRadialMenu
   bInherited=true
   bIsVariable=true
   parent=Overlay_0
   widget=RadialMenu_Picker
   properties {
    entryWidgetClass=&/FortSelectableRadialEntry.FortSelectableRadialEntry_C
    joystickDeadZone=0.4
    mouseDeadZoneRadiusPercentage=0.3
    navigation=&self:WidgetTree.RadialMenu_Picker.WidgetNavigation_0
    orderingMode=Custom
    visibility=Visible
    customOrdering [N,E,W,S,NE,SE,SW,NW]
    radialBoxSettings {angleBetweenItems=0, bClockwiseOrder=false, bDistributeItemsEvenly=true, marginSize=0, sectorCentralAngle=360, startingAngle=90}
   }
   slot {class=OverlaySlot, properties {horizontalAlignment=HAlign_Fill, verticalAlignment=VAlign_Fill}}
  }
 ]
}
